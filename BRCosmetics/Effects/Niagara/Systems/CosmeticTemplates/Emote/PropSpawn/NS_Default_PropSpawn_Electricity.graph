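# Default electricity effect template for emote prop spawns: a short, light-blue radial lightning burst.
# The active LightningSpawn_Radial emitter packs lightning meshes, falling sparks, aura, flare, glow and light into spawn groups, copied to VisibilityTag and animated by per-layer curves.
# Emitter-level color, size and lifetime multipliers drive the layers.
# A vertically stretched LightningSpawn_Long variant with curl-noise motion is included but disabled.
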NiagaraSystem /BRCosmetics/Effects/Niagara/Systems/CosmeticTemplates/Emote/PropSpawn/NS_Default_PropSpawn_Electricity
dependencies {
 usedDataInterfaces [
  &NiagaraDataInterfaceArrayFloat
  &NiagaraDataInterfaceCurve
  &NiagaraDataInterfaceColorCurve
  &NiagaraDataInterfaceVector2DCurve
  &NiagaraDataInterfaceVectorCurve
 ]
 usedDynamicInputs []
 usedModules [&SystemState]
}
emitters {
 LightningSpawn_Long {
  bEnabled=false
  emitterUpdateScript [
   &EmitterState(Life Cycle Mode=::NewEnumerator1, Loop Behavior=::NewEnumerator1, Loop Duration=0.8, UseLoopDelay=false)
   /SetVariables(Emitter.ColorA:LinearColor=<rgba 0.1354 0.6485 1 1>, Emitter.MasterLifetimeMultiplier=1, Emitter.MasterSizeMultiplier=0.6) {isSetParameters=true}
   &SpawnBurst_Instantaneous(Spawn Count=8, Use Spawn Probability=false, UseLoopCountLimit=false)
   &SpawnBurst_Instantaneous(Spawn Group=1, Use Spawn Probability=false, UseLoopCountLimit=false) {module=SpawnBurst_Instantaneous001}
   &SpawnBurst_Instantaneous(Spawn Count=20, Spawn Group=2, Spawn Time=0.08, Use Spawn Probability=false, UseLoopCountLimit=false) {module=SpawnBurst_Instantaneous002}
   &SpawnRate(
    SpawnGroup=3
    SpawnRate=&MakeCustomFloatFromBool(
     Bool=&SetBoolByFloatComparison(A=$System.Age, B=&Multiply_Float(A=0.62, B=$Emitter.MasterLifetimeMultiplier))
     True Float=0
     False Float=80
    )
    Use Spawn Probability=false
   ) {module=SpawnRate002}
   &SpawnBurst_Instantaneous(Spawn Group=4, Use Spawn Probability=false, UseLoopCountLimit=false) {module=SpawnBurst_Instantaneous003}
  ]
  particleSpawnScript [
   &PartitionParticles()
   &InitializeParticle(
    Color=&Multiply_LinearColorByFloat(Linear Color=$Emitter.ColorA, Float=0.4)
    Lifetime=&Multiply_Float(
     A=&MergeHelper_FloatSwitch(
      0=&RandomRangeFloat(Minimum=0.25, Maximum=0.5)
      1=0.35
      2=&RandomRangeFloat(Minimum=0.2, Maximum=0.6)
      3=&RandomRangeFloat(Minimum=0, Maximum=&RandomRangeFloat(Minimum=0.1, Maximum=0.13))
      4=0.55
      Index=&SpawnGroup()
      Num=::NewEnumerator3
     )
     B=$Emitter.MasterLifetimeMultiplier
    )
    Mesh Scale Mode=::NewEnumerator3
    Mesh Uniform Scale=&Multiply_Float(A=&RandomRangeFloat(Minimum=0.1, Maximum=0.12), B=$Emitter.MasterSizeMultiplier)
    Position=<0 0 7>
    Position Mode=::NewEnumerator1
    Sprite Rotation Angle=&MergeHelper_FloatSwitch(0=0, 1=90, 2=?, 3=?, Index=&SpawnGroup(), Num=::NewEnumerator0)
    Sprite Rotation Mode=::NewEnumerator0
    UsePositionOffset=false
    Write Material Random=true
   )
   /SetVariables(Particles.Random=&RandomRangeFloat(Minimum=0, Maximum=1)) {isSetParameters=true}
   &CameraOffset(Camera Offset Amount=10)
   &AddVelocity(
    Velocity=&Multiply_Vector(
     A=&MergeHelper_VectorSwitch(
      0=<0 0 1>
      1=<0 0 0>
      2=&RandomRangeVector(Minimum=<-0.5 -0.5 2>, Maximum=<0.5 0.5 2>)
      3=<0 0 0>
      4=<0 0 0>
      Index=&SpawnGroup()
      Num=::NewEnumerator3
     )
     B=&VectorFromFloat(Value=&SelectFloatFromArray(Direct Array Index=$Particles.PartitionParticles.Partition, Float Selection Array=DataInterface() {propertyValues {FloatData [-1,1]}}))
    )
    Velocity Speed Scale=&Multiply_Float(
     A=&MergeHelper_FloatSwitch(
      0=1
      1=0
      2=&RandomRangeFloat(Minimum=50, Maximum=300)
      3=?
      Index=&SpawnGroup()
      Num=::NewEnumerator1
     )
     B=$Emitter.MasterSizeMultiplier
    )
   ) {module=AddVelocity002}
   &ShapeLocation(
    MaskSpawnGroup=false
    Non Uniform Scale=&MergeHelper_VectorSwitch(0=<1 1 1>, 1=<1 1 1>, 2=<1 1 1>, 3=<0.2 0.2 5>, Index=&SpawnGroup(), Num=::NewEnumerator2)
    Override Seed=false
    Sphere Radius=&Multiply_Float(
     A=&MergeHelper_FloatSwitch(0=18, 1=0, 2=15, 3=10, 4=0, Index=&SpawnGroup(), Num=::NewEnumerator3)
     B=$Emitter.MasterSizeMultiplier
    )
    UseOwnerScale=true
   ) {module=ShapeLocation001}
   /SetVariables(
    Particles.Flare Color:LinearColor=$Emitter.ColorA
    Particles.Flare Sprite Size:Vector2f=&Multiply_Vector2D(A=<400 350>, B=&Vector2DFromFloat(Value=$Emitter.MasterSizeMultiplier))
   ) {isSetParameters=true}
   /SetVariables(Particles.VisibilityTag=&SpawnGroup()) {isSetParameters=true}
   /SetVariables(Particles.Sparks Sprite Size:Vector2f=&Multiply_Vector2D(
     A=&Vector2DFromFloat(Value=&RandomRangeFloat(Minimum=1, Maximum=2))
     B=&Vector2DFromFloat(Value=$Emitter.MasterSizeMultiplier)
    )) {isSetParameters=true}
   /SetVariables(
    Particles.Curved Lightning Color:LinearColor=$Emitter.ColorA
    Particles.Curved Lightning Size:Vector3f=&Multiply_Vector(
     A=&VectorFromFloat(Value=&RandomRangeFloat(Minimum=0.8, Maximum=1))
     B=&VectorFromFloat(Value=$Emitter.MasterSizeMultiplier)
    )
   ) {isSetParameters=true}
   /SetVariables(
    Particles.Lightning Glow Sprite Size:Vector2f=&Multiply_Vector2D(A=<40 145>, B=&Vector2DFromFloat(Value=$Emitter.MasterSizeMultiplier))
    Particles.Lightning Sprite Glow Color:LinearColor=$Emitter.ColorA
   ) {isSetParameters=true}
  ]
  particleUpdateScript [
   &CurlNoiseForce(
    Bias Noise=false
    Noise Strength=&MergeHelper_FloatSwitch(
     0=0
     1=0
     2=&FloatFromCurve(
      FloatCurve=DataInterface() {propertyValues {Curve {keys [{interpMode=RCIM_Cubic, tangentMode=RCTM_User, time=0, value=0}, {interpMode=RCIM_Cubic, tangentMode=RCTM_User, time=1, value=1}]}}}
      Scale Curve=500
     )
     3=&FloatFromCurve(
      FloatCurve=DataInterface() {propertyValues {Curve {keys [{interpMode=RCIM_Cubic, tangentMode=RCTM_User, time=0, value=0}, {interpMode=RCIM_Cubic, tangentMode=RCTM_User, time=1, value=1}]}}}
      Scale Curve=500
     )
     Index=&SpawnGroup()
     Num=::NewEnumerator2
    )
    Pan Noise=true
    Pan Noise Field=&MaskVectorBySpawnGroup(Masked Vector=<0 0 0>, Passthrough Vector=<0 0 1>)
   )
   &ParticleState()
   &ScaleMeshSize(Scale Factor=&RandomRangeVector(Minimum=<2.5 2.5 5>, Maximum=<4.2 4.2 6.5>)) {module=ScaleMeshSize001}
   &ScaleMeshSize(Scale Factor=&MakeVector(
     X=1
     Y=1
     Z=&FloatFromCurve(FloatCurve=DataInterface() {propertyValues {Curve {keys [
          {interpMode=RCIM_Cubic, leaveTangent=5.4877, tangentMode=RCTM_Break, time=0.2, value=0}
          {interpMode=RCIM_Cubic, leaveTangent=-0.1745, time=0.7665, value=1}
         ]}}})
    ))
   &DynamicMaterialParameters(
    Index 0 Param 1=&MakeFloatFromVector(Vector=$Particles.Scale, Channel=::NewEnumerator0)
    Index 0 Param 2=&RandomRangeFloat(Minimum=0, Maximum=10)
    Index 0 Param 3=&FloatFromCurve(
     FloatCurve=DataInterface() {propertyValues {Curve {keys [{interpMode=RCIM_Cubic, leaveTangent=8.1085, tangentMode=RCTM_Break, time=0.8568, value=1}, {tangentMode=RCTM_User, time=1, value=0}]}}}
     Scale Curve=0.05
    )
    Index 0 Param 4=0.8
    Index 1 Param 1=&FloatFromCurve(FloatCurve=DataInterface() {propertyValues {Curve {keys [
         {interpMode=RCIM_Cubic, leaveTangent=1.1034, tangentMode=RCTM_Break, time=0, value=0}
         {arriveTangent=2.2399, tangentMode=RCTM_User, time=0.5, value=1}
        ]}}})
    Index 1 Param 2=&FloatFromCurve(FloatCurve=DataInterface() {propertyValues {Curve {keys [<at 0.5 0>, {arriveTangent=3.2053, tangentMode=RCTM_User, time=0.9, value=1}]}}})
    Index 1 Param 3=&FloatFromCurve(FloatCurve=DataInterface() {propertyValues {Curve {keys [<at 0 0 RCIM_Cubic>, {arriveTangent=3.4223, interpMode=RCIM_Cubic, leaveTangent=3.4223, tangentMode=RCTM_User, time=1, value=1}]}}})
    Index 1 Param 4=&FloatFromCurve(
     FloatCurve=DataInterface() {propertyValues {Curve {keys [
         {interpMode=RCIM_Cubic, leaveTangent=0.1169, tangentMode=RCTM_Break, time=0, value=0}
         {arriveTangent=2.5708, interpMode=RCIM_Cubic, leaveTangent=2.5708, tangentMode=RCTM_User, time=1, value=1}
        ]}}}
     Scale Curve=&RandomRangeFloat(Minimum=30, Maximum=40)
    )
    Index 2 Param 1=&FloatFromCurve(FloatCurve=DataInterface() {propertyValues {Curve {keys [
         {interpMode=RCIM_Cubic, leaveTangent=-0.0025, tangentMode=RCTM_Break, time=0, value=0.5}
         {arriveTangent=-0.0316, interpMode=RCIM_Cubic, tangentMode=RCTM_Break, time=1, value=-0.2}
        ]}}})
    Index 2 Param 2=&FloatFromCurve(FloatCurve=DataInterface() {propertyValues {Curve {keys [
         {interpMode=RCIM_Cubic, tangentMode=RCTM_User, time=0, value=0}
         {arriveTangent=2.0522, interpMode=RCIM_Cubic, leaveTangent=2.0522, tangentMode=RCTM_User, time=1, value=1}
        ]}}})
    Module.Index0Float0Write=true
    Module.Index0Float1Write=true
    Module.Index0Float2Write=true
    Module.Index0Float3Write=true
    Module.Index1Float0Write=true
    Module.Index1Float1Write=false
    Module.Index1Float2Write=true
    Module.Index1Float3Write=true
    Module.Index2Float0Write=true
    Module.Index2Float1Write=true
    Module.Index2Float2Write=false
    Module.Index2Float3Write=false
    Module.Index3Float0Write=false
    Module.Index3Float1Write=false
    Module.Index3Float2Write=false
    Module.Index3Float3Write=true
    Write Parameter Index 1=::NewEnumerator1
    Write Parameter Index 2=::NewEnumerator1
    Write Parameter Index 3=::NewEnumerator1
   )
   &Drag(Drag=&RandomRangeFloat(Minimum=2, Maximum=4), Use Linear Drag=true, Use Rotational Drag=true)
   &SolveForcesAndVelocity(Clamp Velocity=false, Limit Acceleration=false)
   &ScaleColor(
    Scale Alpha=&FloatFromCurve(FloatCurve=DataInterface() {propertyValues {Curve {keys [
         {tangentMode=RCTM_User, time=0, value=1}
         {arriveTangent=-0.0405, interpMode=RCIM_Cubic, leaveTangent=-0.0405, tangentMode=RCTM_User, time=0.4, value=1}
         {arriveTangent=-3.5786, interpMode=RCIM_Cubic, leaveTangent=-3.5786, tangentMode=RCTM_User, time=1, value=0}
        ]}}})
    Scale RGB=&Lerp_Vector(
     A=&VectorFromFloat(Value=&FloatFromCurve(
       FloatCurve=DataInterface() {propertyValues {Curve {keys [<at 0 1>,<at 0.1 1>,<at 0.2 2>,<at 0.3 1>,<at 0.6 1>,<at 0.7 2>,<at 0.8 1>,<at 1 1>]}}}
       Scale Curve=10
      ))
     B=&VectorFromFloat(Value=&RandomRangeFloat(Minimum=0.3, Maximum=5, Evaluation Type=::NewEnumerator1))
     Alpha=&FloatFromCurve(FloatCurve=DataInterface() {propertyValues {Curve {keys [
          {interpMode=RCIM_Cubic, leaveTangent=2.5294, tangentMode=RCTM_Break, time=0, value=0}
          {arriveTangent=-0.2166, interpMode=RCIM_Cubic, tangentMode=RCTM_Break, time=0.4, value=1}
         ]}}})
    )
    ScaleA=true
    ScaleRGB=true
   )
   /SetVariables(Particles.MeshOrientation=&MultiplyQuaternion(
     Quaternion A=&MultiplyQuaternion(
      Quaternion A=&MakeQuatFromAxisAngle(Axis=<0 0 1>, AngleInDegrees=0)
      Quaternion B=&FindQuatBetweenTwoVectors(From Vector=<0 0 1>, To Vector=$Particles.Velocity)
     )
     Quaternion B=<0 0 0 1>
    )) {isSetParameters=true}
   /SetVariables(
    Particles.Flare Color:LinearColor=&MultiplyLinearColor(
     Linear Color=$Particles.Initial.Flare Color
     Scale Factor=&Vector4FromCurve(Vector4Curve=DataInterface() {propertyValues {AlphaCurve {keys [<at 0 0>,<at 0.1264 0.5>,<at 1 0>]}, BlueCurve {keys [<at 0 2>]}, GreenCurve {keys [<at 0 2>]}, RedCurve {keys [<at 0 2>]}}})
    )
    Particles.Flare Sprite Size:Vector2f=&Multiply_Vector2D(
     A=$Particles.Initial.Flare Sprite Size
     B=&Vector2DFromFloat(Value=&FloatFromCurve(FloatCurve=DataInterface() {propertyValues {Curve {keys [<at 0 0>,<at 0.3 1>,<at 1 0>]}}}))
    )
   ) {isSetParameters=true}
   &FN_Light_Attributes(
    Create Exponent Attribute=true
    Create Radius Attribute=true
    Exponent=1
    Radius=&FloatFromCurve(FloatCurve=DataInterface() {propertyValues {Curve {keys [<at 0 0>,<at 0.1 1>,<at 1 0>]}}}, Scale Curve=120)
   )
   /SetVariables(
    Particles.Sparks Color:LinearColor=&MultiplyLinearColor(
     Linear Color=$Particles.Initial.Sparks Color
     Scale Factor=&Vector4FromCurve(Vector4Curve=DataInterface() {propertyValues {
        AlphaCurve {keys [<at 0 15>,<at 0.2737 1>,<at 0.7725 1>,<at 1 0>]}
        BlueCurve {keys [<at 0 10>,<at 0.2664 0>,<at 1 0>]}
        GreenCurve {keys [<at 0 10>,<at 0.2664 1.9553>,<at 1 0>]}
        RedCurve {keys [<at 0 10>,<at 0.2664 5>,<at 1 0>]}
       }})
    )
    Particles.Sparks Sprite Size:Vector2f=&Multiply_Vector2D(
     A=$Particles.Initial.Sparks Sprite Size
     B=&Vector2DFromCurve(Vector2Curve=DataInterface() {propertyValues {XCurve {keys [<at 0 1>,<at 1 0>]}, YCurve {keys [<at 0 1>,<at 1 0>]}}}, ScaleCurve=<0.5 7>)
    )
   ) {isSetParameters=true}
   /SetVariables(Particles.Lightning glow color:LinearColor=&MakeLinearColorFromVectorAndFloat(
     "Vector (RGB)"=&MakeVectorFromLinearColorRGB(Color=&MultiplyLinearColor(
       Linear Color=$Emitter.ColorA
       Scale Factor=/Niagara/DynamicInputs/Multiply/Multiply_Vector4_ByFloat(Vector4:Vector4f=<1 1 1 1>, Float:NiagaraFloat=1)
      ))
     "Float (Alpha)"=&Multiply_Float(
      A=&FloatFromCurve(
       FloatCurve=DataInterface() {propertyValues {Curve {keys [
           <at 0 0 RCIM_Cubic>
           {arriveTangent=-0.0166, interpMode=RCIM_Cubic, leaveTangent=-0.0166, tangentMode=RCTM_User, time=0.1, value=0}
           <at 0.2 0.3 RCIM_Cubic>
           {arriveTangent=-1.0959, interpMode=RCIM_Cubic, leaveTangent=-1.0959, tangentMode=RCTM_User, time=0.8, value=0}
          ]}}}
       Scale Curve=0.4
      )
      B=3
     )
    )) {isSetParameters=true}
   /SetVariables(
    Particles.Glow Dy02:Vector4f=&MakeVector4(
     X=&FloatFromCurve(
      FloatCurve=DataInterface() {propertyValues {Curve {keys [{tangentMode=RCTM_User, time=0, value=1}, {tangentMode=RCTM_User, time=1, value=0}]}}}
      Scale Curve=&Multiply_Float(A=0.6, B=$Emitter.MasterSizeMultiplier)
     )
     Y=0.7
     Z=0
     W=1
    )
    Particles.Glow Dy03:Vector4f=&MakeVector4(X=0.8, Y=0, Z=&Lerp_Float(A=1.3, B=1.8, Alpha=$Particles.Random), W=1)
   ) {isSetParameters=true}
   /SetVariables(
    Particles.Curved Lighting Dyn 1:Vector4f=&MakeVector4(
     X=&FloatFromCurve(FloatCurve=DataInterface() {propertyValues {Curve {keys [{tangentMode=RCTM_User, time=0, value=0}, {tangentMode=RCTM_User, time=1, value=1}]}}})
     Y=0.7
     Z=&FloatFromCurve(FloatCurve=DataInterface() {propertyValues {Curve {keys [<at 0 0 RCIM_Cubic>, {arriveTangent=-0.3684, interpMode=RCIM_Cubic, leaveTangent=-0.3684, tangentMode=RCTM_User, time=1, value=3}]}}})
     W=1
    )
    Particles.Curved Lighting Dyn 2:Vector4f=&MakeVector4(
     X=&Multiply_Float(A=0.5, B=$Emitter.MasterSizeMultiplier)
     Y=&RandomRangeFloat(Minimum=0, Maximum=1)
     Z=&Lerp_Float(A=0.5, B=0.8, Alpha=$Particles.Random)
     W=1
    )
    Particles.Curved Lightning Color:LinearColor=&MultiplyLinearColor(
     Linear Color=$Particles.Initial.Curved Lightning Color
     Scale Factor=&MakeVector4FromVectorAndFloat(
      "Vector (RGB)"=&Multiply_VectorByFloat(
       Vector=<1 1 1>
       Float=&Multiply_Float(A=&RandomRangeFloat(Minimum=0.1, Maximum=1, Evaluation Type=::NewEnumerator1), B=3)
      )
      "Float (Alpha)"=&Multiply_Float(
       A=10
       B=&FloatFromCurve(FloatCurve=DataInterface() {propertyValues {Curve {keys [
            <at 0 0 RCIM_Cubic>
            {arriveTangent=-0.0166, interpMode=RCIM_Cubic, leaveTangent=-0.0166, tangentMode=RCTM_User, time=0.1, value=0}
            <at 0.2 1 RCIM_Cubic>
            <at 1 0 RCIM_Cubic>
           ]}}})
      )
     )
    )
    Particles.Curved Lightning Orientation:Quat4f=&MakeQuatFromAxisAngle(
     Axis=&RandomRangeVector(Minimum=<0 0 0>, Maximum=<1 1 1>)
     AngleInDegrees=&RandomRangeFloat(Minimum=0, Maximum=360)
    )
    Particles.Curved Lightning Size:Vector3f=&Multiply_Vector(
     A=$Particles.Initial.Curved Lightning Size
     B=&VectorFromFloat(Value=&FloatFromCurve(FloatCurve=DataInterface() {propertyValues {Curve {keys [<at 0 0.3>,<at 0.4 1.6>]}}}, CurveIndex=$Emitter.Age))
    )
   ) {isSetParameters=true}
   /SetVariables(
    Particles.Lightning Glow Sprite Size:Vector2f=&Multiply_Vector2D(
     A=$Particles.Initial.Lightning Glow Sprite Size
     B=&Vector2DFromFloat(Value=&FloatFromCurve(FloatCurve=DataInterface() {propertyValues {Curve {keys [<at 0 1>,<at 0.3 1.2>,<at 1 0.56>]}}}))
    )
    Particles.Lightning Sprite Glow Color:LinearColor=&MultiplyLinearColor(
     Linear Color=$Particles.Initial.Lightning Sprite Glow Color
     Scale Factor=&Vector4FromCurve(Vector4Curve=DataInterface() {propertyValues {AlphaCurve {keys [<at 0 0>,<at 0.0533 0.1>,<at 1 0>]}, BlueCurve {keys [<at 0 1>]}, GreenCurve {keys [<at 0 1>]}, RedCurve {keys [<at 0 1>]}}})
    )
   ) {isSetParameters=true}
  ]
  rendererClasses [
   &NiagaraMeshRendererProperties
   &NiagaraSpriteRendererProperties
   &NiagaraLightRendererProperties
  ]
  renderers [
   {rendererClass=&NiagaraMeshRendererProperties, rendererIndex=0}
   {rendererClass=&NiagaraMeshRendererProperties, rendererIndex=1}
   {rendererClass=&NiagaraSpriteRendererProperties, rendererIndex=2}
   {rendererClass=&NiagaraLightRendererProperties, rendererIndex=3}
   {rendererClass=&NiagaraSpriteRendererProperties, rendererIndex=4}
   {rendererClass=&NiagaraMeshRendererProperties, rendererIndex=5}
   {rendererClass=&NiagaraMeshRendererProperties, rendererIndex=6}
   {rendererClass=&NiagaraSpriteRendererProperties, rendererIndex=7}
  ]
 }
 LightningSpawn_Radial {
  emitterUpdateScript [
   &EmitterState(Life Cycle Mode=::NewEnumerator1, Loop Behavior=::NewEnumerator1, Loop Duration=0.6, UseLoopDelay=false)
   /SetVariables(
    Emitter.ColorA:LinearColor=<rgba 0.1354 0.6485 1 1>
    Emitter.MasterLifetimeMultiplier=1
    Emitter.MasterSizeMultiplier=0.7
    Emitter.ShapeLocationScale:Vector3f=<1 1 1>
   ) {isSetParameters=true}
   &SpawnRate(
    SpawnRate=&MakeCustomFloatFromBool(
     Bool=&SetBoolByFloatComparison(A=$System.Age, B=&Multiply_Float(A=0.55, B=$Emitter.MasterLifetimeMultiplier))
     True Float=0
     False Float=60
    )
    Use Spawn Probability=true
   )
   &SpawnBurst_Instantaneous(Spawn Count=7, Spawn Time=0.15, Use Spawn Probability=false, UseLoopCountLimit=false) {module=SpawnBurst_Instantaneous001}
   &SpawnRate(
    SpawnGroup=1
    SpawnRate=&MakeCustomFloatFromBool(
     Bool=&SetBoolByFloatComparison(A=$System.Age, B=&Multiply_Float(A=0.19, B=$Emitter.MasterLifetimeMultiplier))
     True Float=0
     False Float=30
    )
    Use Spawn Probability=false
   ) {module=SpawnRate001}
   &SpawnBurst_Instantaneous(
    Spawn Count=&RandomRangeInt(Minimum=3, Maximum=4)
    Spawn Group=1
    Use Spawn Probability=false
    UseLoopCountLimit=false
   ) {module=SpawnBurst_Instantaneous002}
   &SpawnBurst_Instantaneous(Spawn Count=65, Spawn Group=2, Use Spawn Probability=false, UseLoopCountLimit=false)
   &SpawnBurst_Instantaneous(Spawn Group=3, Use Spawn Probability=false, UseLoopCountLimit=false) {module=SpawnBurst_Instantaneous003}
   &SpawnBurst_Instantaneous(Spawn Group=4, Use Spawn Probability=false, UseLoopCountLimit=false) {module=SpawnBurst_Instantaneous004}
   &SpawnBurst_Instantaneous(Spawn Group=5, Use Spawn Probability=false, UseLoopCountLimit=false) {module=SpawnBurst_Instantaneous005}
  ]
  particleSpawnScript [
   &InitializeParticle(
    Color=$Emitter.ColorA
    Lifetime=&Multiply_Float(
     A=&MergeHelper_FloatSwitch(
      0=&RandomRangeFloat(Minimum=0.1, Maximum=0.15)
      1=&RandomRangeFloat(Minimum=0.3, Maximum=0.45)
      2=&RandomRangeFloat(Minimum=0.3, Maximum=0.8)
      3=0.25
      4=0.22
      5=0.6
      Index=&SpawnGroup()
      Num=::NewEnumerator4
     )
     B=$Emitter.MasterLifetimeMultiplier
    )
    Mesh Scale Mode=::NewEnumerator3
    Mesh Uniform Scale=&Multiply_Float(A=&RandomRangeFloat(Minimum=1, Maximum=1.2), B=$Emitter.MasterSizeMultiplier)
    UsePositionOffset=false
    Write Material Random=true
   )
   &CameraOffset(Camera Offset Amount=&MergeHelper_FloatSwitch(0=5, 1=5, 2=0, 3=0, 4=10, 5=0, Index=&SpawnGroup(), Num=::NewEnumerator4))
   &InitialMeshOrientation(Rotation=&RandomRangeVector(Minimum=<0 0 0>, Maximum=<1 1 1>), Use Rotation Vector=true)
   &AddVelocity(
    Module.Use Falloff Distance=false
    Override Seed=false
    Velocity Mode=::NewEnumerator1
    Velocity Speed=&Multiply_Float(
     A=&MergeHelper_FloatSwitch(
      0=10
      1=0.01
      2=&RandomRangeFloat(Minimum=-600, Maximum=600)
      3=?
      Index=&SpawnGroup()
      Num=::NewEnumerator1
     )
     B=$Emitter.MasterSizeMultiplier
    )
   )
   &ShapeLocation(
    MaskSpawnGroup=false
    Non Uniform Scale=&Multiply_Vector(A=<1 1 1>, B=$Emitter.ShapeLocationScale)
    Override Seed=false
    Sphere Radius=&Multiply_Float(
     A=&MergeHelper_FloatSwitch(0=5, 1=0, 2=0, 3=0, 4=0, 5=0, Index=&SpawnGroup(), Num=::NewEnumerator4)
     B=$Emitter.MasterSizeMultiplier
    )
    UseOwnerScale=true
   )
   /SetVariables(Particles.Random=&RandomRangeFloat(Minimum=0, Maximum=1)) {isSetParameters=true}
   /SetVariables(Particles.VisibilityTag=&SpawnGroup()) {isSetParameters=true}
   /SetVariables(
    Particles.Branch Lighting Size:Vector3f=&Multiply_Vector(
     A=&RandomRangeVector(Minimum=<1 1 0.85>, Maximum=<1.5 1.5 1.35>)
     B=&VectorFromFloat(Value=&Multiply_Float(A=0.75, B=$Emitter.MasterSizeMultiplier))
    )
    Particles.Branch Lightning Color:LinearColor=$Emitter.ColorA
   ) {isSetParameters=true}
   /SetVariables(
    Particles.Sparks Color:LinearColor=<rgba 5 5 5 0.01>
    Particles.Sparks Color 2:LinearColor=&MultiplyLinearColor(Linear Color=$Emitter.ColorA, Scale Factor=&MakeVector4_FromLinearColor(LinearColor=<rgba 5 5 5 0.004>))
    Particles.Sparks Sprite Size:Vector2f=&Multiply_Vector2D(A=<4 2>, B=&Vector2DFromFloat(Value=$Emitter.MasterSizeMultiplier))
    Particles.Sparks Sprite Size 2:Vector2f=&Multiply_Vector2D(A=<20 12>, B=&Vector2DFromFloat(Value=$Emitter.MasterSizeMultiplier))
   ) {isSetParameters=true}
   /SetVariables(
    Particles.Aura Color:LinearColor=$Emitter.ColorA
    Particles.Aura Sprite Size:Vector2f=&Vector2DFromFloat(Value=&Multiply_Float(A=150, B=$Emitter.MasterSizeMultiplier))
   ) {isSetParameters=true}
   /SetVariables(
    Particles.Flare Color:LinearColor=$Emitter.ColorA
    Particles.Flare Sprite Size:Vector2f=&Vector2DFromFloat(Value=&Multiply_Float(A=297.7988, B=$Emitter.MasterSizeMultiplier))
   ) {isSetParameters=true}
   /SetVariables(
    Particles.Lightning Glow Sprite Size:Vector2f=&Multiply_Vector2D(
     A=&Vector2DFromFloat(Value=90)
     B=&Vector2DFromFloat(Value=$Emitter.MasterSizeMultiplier)
    )
    Particles.Lightning Sprite Glow Color:LinearColor=$Emitter.ColorA
   ) {isSetParameters=true}
  ]
  particleUpdateScript [
   &ParticleState()
   &ScaleSpriteSize(Uniform Curve Sprite Scale=DataInterface() {propertyValues {Curve {keys [{tangentMode=RCTM_User, time=0, value=0.5}, {tangentMode=RCTM_User, time=1, value=1}]}}})
   &UpdateMeshOrientation(
    Rotation Rate=&MergeHelper_FloatSwitch(0=1, 1=0, 2=?, 3=?, Index=&SpawnGroup(), Num=::NewEnumerator0)
    Rotation Vector=&MergeHelper_VectorSwitch(
     0=&RandomRangeVector(Minimum=<0 0.1 0>, Maximum=<0 0.6 0>)
     1=<0 0 0>
     2=?
     3=?
     Index=&SpawnGroup()
     Num=::NewEnumerator0
    )
   )
   &ScaleMeshSize(Scale Factor=&VectorFromFloat(Value=&FloatFromCurve(
      FloatCurve=DataInterface() {propertyValues {Curve {keys [<at 0 0.3>,<at 0.4 1.6>]}}}
      CurveIndex=$Emitter.Age
      Scale Curve=&RandomRangeFloat(Minimum=1, Maximum=1.2)
     )))
   &ScaleColor(
    Scale Alpha=&Multiply_Float(
     A=10
     B=&FloatFromCurve(FloatCurve=DataInterface() {propertyValues {Curve {keys [
          <at 0 0 RCIM_Cubic>
          {arriveTangent=-0.0166, interpMode=RCIM_Cubic, leaveTangent=-0.0166, tangentMode=RCTM_User, time=0.1, value=0}
          <at 0.2 1 RCIM_Cubic>
          <at 1 0 RCIM_Cubic>
         ]}}})
    )
    Scale RGB=&Multiply_VectorByFloat(
     Vector=<1 1 1>
     Float=&Multiply_Float(A=&RandomRangeFloat(Minimum=0.1, Maximum=1, Evaluation Type=::NewEnumerator1), B=5)
    )
    ScaleA=true
    ScaleRGB=true
   )
   &DynamicMaterialParameters(
    Index 0 Param 1=&FloatFromCurve(FloatCurve=DataInterface() {propertyValues {Curve {keys [{tangentMode=RCTM_User, time=0, value=0}, {tangentMode=RCTM_User, time=1, value=1}]}}})
    Index 0 Param 2=0.7
    Index 0 Param 3=&FloatFromCurve(FloatCurve=DataInterface() {propertyValues {Curve {keys [<at 0 0 RCIM_Cubic>, {arriveTangent=-0.3684, interpMode=RCIM_Cubic, leaveTangent=-0.3684, tangentMode=RCTM_User, time=1, value=3}]}}})
    Index 0 Param 4=1
    Index 1 Param 1=&Multiply_Float(A=0.5, B=$Emitter.MasterSizeMultiplier)
    Index 1 Param 2=&RandomRangeFloat(Minimum=0, Maximum=1)
    Index 1 Param 3=&Lerp_Float(A=0.6, B=0.8, Alpha=$Particles.Random)
    Index 1 Param 4=1
    Module.Index0Float0Write=true
    Module.Index0Float1Write=true
    Module.Index0Float2Write=true
    Module.Index0Float3Write=true
    Module.Index1Float0Write=true
    Module.Index1Float1Write=true
    Module.Index1Float2Write=true
    Module.Index1Float3Write=true
    Module.Index3Float0Write=true
    Module.Index3Float1Write=true
    Module.Index3Float2Write=true
    Module.Index3Float3Write=true
    Write Parameter Index 1=::NewEnumerator1
    Write Parameter Index 3=::NewEnumerator1
   )
   &GravityForce(Gravity=&MergeHelper_VectorSwitch(0=<0 0 0>, 1=<0 0 0>, 2=<0 0 -980>, 3=<0 0 0>, 4=<0 0 0>, 5=<0 0 0>, Index=&SpawnGroup(), Num=::NewEnumerator4))
   &SolveForcesAndVelocity(Clamp Velocity=false, Limit Acceleration=false)
   /SetVariables(Particles.Lightning glow color:LinearColor=&MakeLinearColorFromVectorAndFloat(
     "Vector (RGB)"=&MakeVectorFromLinearColorRGB(Color=&MultiplyLinearColor(
       Linear Color=$Emitter.ColorA
       Scale Factor=&Multiply_Vector4_ByFloat(Vector4:Vector4f=<1 1 1 1>, Float:NiagaraFloat=1)
      ))
     "Float (Alpha)"=&Multiply_Float(
      A=&FloatFromCurve(FloatCurve=DataInterface() {propertyValues {Curve {keys [
           <at 0 0 RCIM_Cubic>
           {arriveTangent=-0.0166, interpMode=RCIM_Cubic, leaveTangent=-0.0166, tangentMode=RCTM_User, time=0.1, value=0}
           <at 0.2 0.3 RCIM_Cubic>
           {arriveTangent=-0.9699, interpMode=RCIM_Cubic, leaveTangent=-0.9699, tangentMode=RCTM_User, time=0.8, value=0}
          ]}}})
      B=3
     )
    )) {isSetParameters=true}
   /SetVariables(
    Particles.Glow Dy02:Vector4f=&MakeVector4(
     X=&FloatFromCurve(FloatCurve=DataInterface() {propertyValues {Curve {keys [<at 0 0>,<at 0.2 0.5>,<at 1 0>]}}})
     Y=0.7
     Z=0
     W=1
    )
    Particles.Glow Dy03:Vector4f=&MakeVector4(
     X=&FloatFromCurve(
      FloatCurve=DataInterface() {propertyValues {Curve {keys [{tangentMode=RCTM_User, time=0, value=1}, {tangentMode=RCTM_User, time=1, value=0}]}}}
      Scale Curve=&Multiply_Float(A=1.5, B=$Emitter.MasterSizeMultiplier)
     )
     Y=0.7
     Z=0
     W=0
    )
   ) {isSetParameters=true}
   /SetVariables(
    Particles.Branch Lighting Dyn 1:Vector4f=&MakeVector4(
     X=0.01
     Y=&RandomRangeFloat(Minimum=7, Maximum=9)
     Z=&FloatFromCurve(
      FloatCurve=DataInterface() {propertyValues {Curve {keys [<at 0 1>,<at 1 0>]}}}
      Scale Curve=&Multiply_Float(A=&RandomRangeFloat(Minimum=1, Maximum=1.5), B=$Emitter.MasterSizeMultiplier)
     )
     W=&RandomRangeFloat(Minimum=0.2, Maximum=0.6)
    )
    Particles.Branch Lighting Dyn 2:Vector4f=&MakeVector4(
     X=&FloatFromCurve(FloatCurve=DataInterface() {propertyValues {Curve {keys [<at 0.5 0.5>,<at 1 1>]}}})
     Y=&FloatFromCurve(FloatCurve=DataInterface() {propertyValues {Curve {keys [<at 0 0>,<at 1 1>]}}})
     Z=&FloatFromCurve(FloatCurve=DataInterface() {propertyValues {Curve {keys [<at 0 0 RCIM_Cubic>,<at 1 1 RCIM_Cubic>]}}})
     W=2
    )
    Particles.Branch Lighting Size:Vector3f=&Multiply_Vector(
     A=$Particles.Initial.Branch Lighting Size
     B=&VectorFromCurve(VectorCurve=DataInterface() {propertyValues {
        XCurve {keys [<at 0 0.8>,<at 1 1>]}
        YCurve {keys [<at 0 0.8>,<at 0.8 1>,<at 1 0>]}
        ZCurve {keys [
          {arriveTangent=3.6643, interpMode=RCIM_Cubic, leaveTangent=3.6643, tangentMode=RCTM_User, time=0, value=0.5}
          {arriveTangent=0.5785, interpMode=RCIM_Cubic, leaveTangent=0.5785, tangentMode=RCTM_User, time=0.2, value=0.8706}
          <at 1 1 RCIM_Cubic>
         ]}
       }})
    )
    Particles.Branch Lightning Color:LinearColor=&MultiplyLinearColor(
     Linear Color=$Particles.Initial.Branch Lightning Color
     Scale Factor=&Vector4FromCurve(Vector4Curve=DataInterface() {propertyValues {
        AlphaCurve {keys [<at 0 0>,<at 0.1557 1>,<at 1 0>]}
        BlueCurve {keys [<at 0 11>,<at 1 1>]}
        GreenCurve {keys [<at 0 11>,<at 1 1>]}
        RedCurve {keys [<at 0 11>,<at 1 1>]}
       }})
    )
    Particles.Branch Lightning Dyn 3:Vector4f=&MakeVector4(
     X=&FloatFromCurve(FloatCurve=DataInterface() {propertyValues {Curve {keys [<at 0 0.5>,<at 1 0>]}}})
     Y=1
     Z=0
     W=0
    )
   ) {isSetParameters=true}
   /SetVariables(
    Particles.Glow Lightning Dyn 1:Vector4f=&MakeVector4(
     X=0.05
     Y=&RandomRangeFloat(Minimum=7, Maximum=9)
     Z=&FloatFromCurve(
      FloatCurve=DataInterface() {propertyValues {Curve {keys [<at 0 1>,<at 1 0>]}}}
      Scale Curve=&Multiply_Float(A=1.5, B=$Emitter.MasterSizeMultiplier)
     )
     W=&RandomRangeFloat(Minimum=0.2, Maximum=0.6)
    )
    Particles.Lightning glow color 2:LinearColor=&MultiplyLinearColor(
     Linear Color=$Emitter.ColorA
     Scale Factor=&Vector4FromCurve(Vector4Curve=DataInterface() {propertyValues {
        AlphaCurve {keys [<at 0 0>,<at 0.1557 0.15>,<at 0.5922 0.1>,<at 1 0>]}
        BlueCurve {keys [<at 0 0.4>]}
        GreenCurve {keys [<at 0 0.4>]}
        RedCurve {keys [<at 0 0.4>]}
       }})
    )
   ) {isSetParameters=true}
   /SetVariables(
    Particles.Sparks Color:LinearColor=&MultiplyLinearColor(
     Linear Color=$Particles.Initial.Sparks Color
     Scale Factor=&Vector4FromCurve(Vector4Curve=DataInterface() {propertyValues {
        AlphaCurve {keys [<at 0 15>,<at 0.2737 1>,<at 0.7725 1>,<at 1 0>]}
        BlueCurve {keys [<at 0 15>,<at 0.2664 0>,<at 1 0>]}
        GreenCurve {keys [<at 0 15>,<at 0.2664 1.9553>,<at 1 0>]}
        RedCurve {keys [<at 0 15>,<at 0.2664 5>,<at 1 0>]}
       }})
    )
    Particles.Sparks Color 2:LinearColor=&MultiplyLinearColor(
     Linear Color=$Particles.Initial.Sparks Color 2
     Scale Factor=&Vector4FromCurve(Vector4Curve=DataInterface() {propertyValues {
        AlphaCurve {keys [<at 0 15>,<at 0.2737 1>,<at 0.7725 1>,<at 1 0>]}
        BlueCurve {keys [<at 0 15>,<at 0.2664 0>,<at 1 0>]}
        GreenCurve {keys [<at 0 15>,<at 0.2664 1.9553>,<at 1 0>]}
        RedCurve {keys [<at 0 15>,<at 0.2664 5>,<at 1 0>]}
       }})
    )
    Particles.Sparks Sprite Size:Vector2f=&Multiply_Vector2D(
     A=$Particles.Initial.Sparks Sprite Size
     B=&Vector2DFromCurve(
      Vector2Curve=DataInterface() {propertyValues {
        XCurve {keys [
          {arriveTangent=-2.9073, interpMode=RCIM_Cubic, leaveTangent=-2.9073, leaveTangentWeight=0.3333, tangentMode=RCTM_User, time=0, value=1}
          {arriveTangentWeight=1.3121, interpMode=RCIM_Cubic, tangentMode=RCTM_User, time=1, value=0}
         ]}
        YCurve {keys [
          {arriveTangent=-2.9073, interpMode=RCIM_Cubic, leaveTangent=-2.9073, leaveTangentWeight=0.3333, tangentMode=RCTM_User, time=0, value=1}
          {arriveTangentWeight=1.3121, interpMode=RCIM_Cubic, tangentMode=RCTM_User, time=1, value=0}
         ]}
       }}
      ScaleCurve=<1 15>
     )
    )
    Particles.Sparks Sprite Size 2:Vector2f=&Multiply_Vector2D(
     A=$Particles.Initial.Sparks Sprite Size 2
     B=&Vector2DFromCurve(
      Vector2Curve=DataInterface() {propertyValues {
        XCurve {keys [
          {arriveTangent=-2.9073, interpMode=RCIM_Cubic, leaveTangent=-2.9073, leaveTangentWeight=0.3333, tangentMode=RCTM_User, time=0, value=1}
          {arriveTangentWeight=1.3121, interpMode=RCIM_Cubic, tangentMode=RCTM_User, time=1, value=0}
         ]}
        YCurve {keys [
          {arriveTangent=-2.9073, interpMode=RCIM_Cubic, leaveTangent=-2.9073, leaveTangentWeight=0.3333, tangentMode=RCTM_User, time=0, value=1}
          {arriveTangentWeight=1.3121, interpMode=RCIM_Cubic, tangentMode=RCTM_User, time=1, value=0}
         ]}
       }}
      ScaleCurve=<1 8>
     )
    )
   ) {isSetParameters=true}
   /SetVariables(
    Particles.Aura Color:LinearColor=&MultiplyLinearColor(
     Linear Color=$Particles.Initial.Aura Color
     Scale Factor=&Vector4FromCurve(Vector4Curve=DataInterface() {propertyValues {AlphaCurve {keys [<at 0 0>,<at 0.1068 1>,<at 1 0>]}, BlueCurve {keys [<at 0 3>]}, GreenCurve {keys [<at 0 3>]}, RedCurve {keys [<at 0 3>]}}})
    )
    Particles.Aura Sprite Size:Vector2f=&Multiply_Vector2D(
     A=$Particles.Initial.Aura Sprite Size
     B=&Vector2DFromFloat(Value=&FloatFromCurve(FloatCurve=DataInterface() {propertyValues {Curve {keys [
           {arriveTangent=1.3953, interpMode=RCIM_Cubic, leaveTangent=1.3953, tangentMode=RCTM_User, time=0, value=0.2}
           {arriveTangent=0.8512, interpMode=RCIM_Cubic, leaveTangent=0.8512, tangentMode=RCTM_User, time=1, value=1}
          ]}}}))
    )
   ) {isSetParameters=true}
   /SetVariables(
    Particles.Flare Color:LinearColor=&MultiplyLinearColor(
     Linear Color=$Particles.Initial.Flare Color
     Scale Factor=&Vector4FromCurve(Vector4Curve=DataInterface() {propertyValues {AlphaCurve {keys [<at 0 0>,<at 0.1264 1>,<at 1 0>]}, BlueCurve {keys [<at 0 2>]}, GreenCurve {keys [<at 0 2>]}, RedCurve {keys [<at 0 2>]}}})
    )
    Particles.Flare Sprite Size:Vector2f=&Multiply_Vector2D(
     A=$Particles.Initial.Flare Sprite Size
     B=&Vector2DFromFloat(Value=&FloatFromCurve(FloatCurve=DataInterface() {propertyValues {Curve {keys [<at 0 1>,<at 0.3 1.2>,<at 1 1>]}}}))
    )
   ) {isSetParameters=true}
   &FN_Light_Attributes(
    Create Exponent Attribute=true
    Create Radius Attribute=true
    Exponent=1
    Radius=&FloatFromCurve(FloatCurve=DataInterface() {propertyValues {Curve {keys [<at 0 0>,<at 0.1 1>,<at 1 0>]}}}, Scale Curve=120)
   )
   /SetVariables(
    Particles.Lightning Glow Sprite Size:Vector2f=&Multiply_Vector2D(
     A=$Particles.Initial.Lightning Glow Sprite Size
     B=&Vector2DFromFloat(Value=&FloatFromCurve(FloatCurve=DataInterface() {propertyValues {Curve {keys [<at 0 1>,<at 0.3 1.2>,<at 1 0.56>]}}}))
    )
    Particles.Lightning Sprite Glow Color:LinearColor=&MultiplyLinearColor(
     Linear Color=$Particles.Initial.Lightning Sprite Glow Color
     Scale Factor=&Vector4FromCurve(Vector4Curve=DataInterface() {propertyValues {AlphaCurve {keys [<at 0 0>,<at 0.0533 0.2>,<at 1 0>]}, BlueCurve {keys [<at 0 1>]}, GreenCurve {keys [<at 0 1>]}, RedCurve {keys [<at 0 1>]}}})
    )
   ) {isSetParameters=true}
  ]
  rendererClasses [
   &NiagaraMeshRendererProperties
   &NiagaraSpriteRendererProperties
   &NiagaraLightRendererProperties
  ]
  renderers [
   {rendererClass=&NiagaraMeshRendererProperties, rendererIndex=0}
   {rendererClass=&NiagaraMeshRendererProperties, rendererIndex=1}
   {rendererClass=&NiagaraMeshRendererProperties, rendererIndex=2}
   {rendererClass=&NiagaraMeshRendererProperties, rendererIndex=3}
   {rendererClass=&NiagaraSpriteRendererProperties, rendererIndex=4}
   {rendererClass=&NiagaraSpriteRendererProperties, rendererIndex=5}
   {rendererClass=&NiagaraSpriteRendererProperties, rendererIndex=6}
   {rendererClass=&NiagaraLightRendererProperties, rendererIndex=7}
   {rendererClass=&NiagaraSpriteRendererProperties, rendererIndex=8}
   {rendererClass=&NiagaraSpriteRendererProperties, rendererIndex=9}
  ]
 }
}
summary {
 systemName=NS_Default_PropSpawn_Electricity
 emitters [
  {
   bEnabled=false
   emitterName=LightningSpawn_Long
   rendererClasses [
    &NiagaraMeshRendererProperties
    &NiagaraSpriteRendererProperties
    &NiagaraLightRendererProperties
   ]
  }
  {
   emitterName=LightningSpawn_Radial
   rendererClasses [
    &NiagaraMeshRendererProperties
    &NiagaraSpriteRendererProperties
    &NiagaraLightRendererProperties
   ]
  }
 ]
 userVariables [
  User.bIsFrontEndPreview {type=NiagaraBool}
  User.NormalizedNotifyProgress {type=NiagaraFloat, defaultValue=0, description=""}
 ]
}
