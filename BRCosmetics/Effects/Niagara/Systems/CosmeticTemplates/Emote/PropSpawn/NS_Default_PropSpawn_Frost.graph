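# Niagara cosmetic template for a short frost burst, named as the default prop-spawn effect for emotes.
# One emitter fires five instantaneous spawn groups, each tagged through VisibilityTag, with five sprite renderers.
# Spawn group index switches lifetime, shape, velocity, drag, noise and gravity, while five attribute modules set each layer's look.
# Master color, size and lifetime parameters drive the layers, and a time-of-day NPC curve tints their colors.
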NiagaraSystem /BRCosmetics/Effects/Niagara/Systems/CosmeticTemplates/Emote/PropSpawn/NS_Default_PropSpawn_Frost
dependencies {
 usedDataInterfaces [
  &NiagaraDataInterfaceCurve
  &NiagaraDataInterfaceArrayFloat3
  &NiagaraDataInterfaceSpriteRendererInfo
  &NiagaraDataInterfaceColorCurve
  &NiagaraDataInterfaceArrayFloat
 ]
 usedDynamicInputs []
 usedModules [&SystemState]
}
emitters {PropSpawn_Frost {
  emitterSpawnScript [/SetVariables(
    Emitter.MasterColorA:LinearColor=<rgba 0.2903 0.6905 1 1>
    Emitter.MasterColorB:LinearColor=<rgba 1.4317 2 1.8338 1>
    Emitter.MasterLifetimeMultiplier=1
    Emitter.MasterSizeMultiplier=0.7
   ) {isSetParameters=true}]
  emitterUpdateScript [
   &EmitterState(Life Cycle Mode=::NewEnumerator1, Loop Behavior=::NewEnumerator1, Loop Duration=1, UseLoopDelay=false)
   &SpawnBurst_Instantaneous(Spawn Count=13, Use Spawn Probability=false, UseLoopCountLimit=false)
   &SpawnBurst_Instantaneous(Spawn Group=1, Use Spawn Probability=false, UseLoopCountLimit=false) {module=SpawnBurst_Instantaneous001}
   &SpawnBurst_Instantaneous(Spawn Count=10, Spawn Group=2, Use Spawn Probability=false, UseLoopCountLimit=false) {module=SpawnBurst_Instantaneous002}
   &SpawnBurst_Instantaneous(Spawn Count=13, Spawn Group=3, Use Spawn Probability=false, UseLoopCountLimit=false) {module=SpawnBurst_Instantaneous003}
   &SpawnBurst_Instantaneous(Spawn Count=6, Spawn Group=4, Use Spawn Probability=false, UseLoopCountLimit=false) {module=SpawnBurst_Instantaneous004}
  ]
  particleSpawnScript [
   /SetVariables(Particles.VisibilityTag=&SpawnGroup()) {isSetParameters=true}
   &InitializeParticle(
    Color Mode=::NewEnumerator0
    Lifetime=&Multiply_Float(
     A=$Emitter.MasterLifetimeMultiplier
     B=&MergeHelper_FloatSwitch(
      0=&RandomRangeFloat(Minimum=0.45, Maximum=0.35)
      1=0.5
      2=&RandomRangeFloat(Minimum=0.5, Maximum=0.8)
      3=&RandomRangeFloat(Minimum=0.5, Maximum=0.8)
      4=&RandomRangeFloat(Minimum=0.3, Maximum=0.85)
      Index=&SpawnGroup()
      Num=::NewEnumerator3
     )
    )
    Sprite Rotation Mode=::NewEnumerator2
    UsePositionOffset=false
    Write Material Random=true
   )
   &ShapeLocation(
    MaskSpawnGroup=false
    Override Seed=false
    Sphere Radius=&Multiply_Float(
     A=$Emitter.MasterSizeMultiplier
     B=&MergeHelper_FloatSwitch(0=10, 1=0, 2=30, 3=30, 4=30, Index=&SpawnGroup(), Num=::NewEnumerator3)
    )
    UseOwnerScale=true
   )
   &AddVelocity(
    Module.Use Falloff Distance=false
    Override Seed=false
    Velocity Mode=::NewEnumerator1
    Velocity Speed=&Multiply_Float(
     A=$Emitter.MasterSizeMultiplier
     B=&MergeHelper_FloatSwitch(
      0=&RandomRangeFloat(Minimum=50, Maximum=100)
      1=0
      2=&RandomRangeFloat(Minimum=50, Maximum=200)
      3=&RandomRangeFloat(Minimum=60, Maximum=150)
      4=&RandomRangeFloat(Minimum=300, Maximum=250)
      Index=&SpawnGroup()
      Num=::NewEnumerator3
     )
    )
   )
   &CameraOffset(Camera Offset Amount=10)
   /SetVariables(Particles.FrostDynParam:Vector4f=&MakeVector4(X=2, Y=0, Z=0, W=1)) {isSetParameters=true}
  ]
  particleUpdateScript [
   &ParticleState()
   &Drag(
    Drag=&MergeHelper_FloatSwitch(
     0=&RandomRangeFloat(Minimum=1, Maximum=3)
     1=0
     2=&FloatFromCurve(
      FloatCurve=DataInterface() {propertyValues {Curve {
         defaultValue=1
         keys [
          {arriveTangent=12.7441, interpMode=RCIM_Cubic, leaveTangent=12.7441, tangentMode=RCTM_User, time=0.0618, value=0.4371}
          {arriveTangent=0.0271, interpMode=RCIM_Cubic, leaveTangent=0.0271, tangentMode=RCTM_User, time=0.1754, value=1}
         ]
        }}}
      Scale Curve=&RandomRangeFloat(Minimum=3, Maximum=5)
     )
     3=&FloatFromCurve(
      FloatCurve=DataInterface() {propertyValues {Curve {
         defaultValue=1
         keys [
          {arriveTangent=12.7441, interpMode=RCIM_Cubic, leaveTangent=12.7441, tangentMode=RCTM_User, time=0.0618, value=0.4371}
          {arriveTangent=0.0271, interpMode=RCIM_Cubic, leaveTangent=0.0271, tangentMode=RCTM_User, time=0.1754, value=1}
         ]
        }}}
      Scale Curve=&RandomRangeFloat(Minimum=3, Maximum=5)
     )
     4=&FloatFromCurve(
      FloatCurve=DataInterface() {propertyValues {Curve {
         defaultValue=1
         keys [
          {arriveTangent=1.1082, interpMode=RCIM_Cubic, leaveTangent=1.1082, tangentMode=RCTM_User, time=0.0026, value=0}
          {arriveTangent=0.0271, interpMode=RCIM_Cubic, leaveTangent=0.0271, tangentMode=RCTM_User, time=0.0104, value=1}
         ]
        }}}
      Scale Curve=&RandomRangeFloat(Minimum=7, Maximum=9)
     )
     Index=&SpawnGroup()
     Num=::NewEnumerator3
    )
    Use Linear Drag=true
    Use Rotational Drag=true
   )
   &CurlNoiseForce(
    Bias Noise=false
    Noise Frequency=30
    Noise Strength=&MergeHelper_FloatSwitch(0=0, 1=0, 2=50, 3=150, 4=300, Index=&SpawnGroup(), Num=::NewEnumerator3)
    Pan Noise=false
   )
   &GravityForce(Gravity=&MergeHelper_VectorSwitch(
     0=<0 0 0>
     1=<0 0 0>
     2=<0 0 0>
     3=<0 0 0>
     4=&RandomRangeVector(Minimum=<0 0 -100>, Maximum=<0 0 -300>)
     Index=&SpawnGroup()
     Num=::NewEnumerator3
    ))
   &MeshRotationForce(Rotation=&MaskVector(
     Mask Source=0
     Mask Parameter=?
     Mask=4
     Masked Vector=&SelectVectorFromArray(
      Array Sampling Mode=::NewEnumerator0
      Direct Array Index=?
      Vector Selection Array=DataInterface() {propertyValues {FloatData [<20 20 20>, <-25 -25 -25>, <25 25 25>, <-20 -20 -20>, <10 10 10>]}}
      Random Seed=0
     )
     Passthrough Vector=&MaskVector(
      Mask Source=0
      Mask Parameter=?
      Mask=2
      Masked Vector=&SelectVectorFromArray(
       Array Sampling Mode=::NewEnumerator0
       Direct Array Index=?
       Vector Selection Array=DataInterface() {propertyValues {FloatData [<20 20 20>, <-25 -25 -25>, <25 25 25>, <-20 -20 -20>, <10 10 10>]}}
       Random Seed=0
      )
      Passthrough Vector=<0 0 0>
     )
    ))
   &SolveForcesAndVelocity(Clamp Velocity=false, Limit Acceleration=false)
   &SubUVAnimation(
    End Frame Range Override=30
    "Sprite Renderer "=DataInterface() {propertyValues {SpriteRenderer=&self:PropSpawn_Frost_0.NiagaraSpriteRendererProperties_0}}
    UseEndFrame=true
    UseStartFrame=true
   )
   /SetVariables() {isSetParameters=true}
   /SetVariables(Particles.TODColor:LinearColor=&ColorFromCurve(
     ColorCurve=DataInterface() {propertyValues {
       AlphaCurve {keys [<at 0 1 RCIM_Cubic>]}
       BlueCurve {keys [
         <at 0.0043 0.7>
         {arriveTangent=-0.1359, interpMode=RCIM_Cubic, leaveTangent=-0.1359, time=0.1789, value=0.7}
         <at 0.2812 0.6624>
         {arriveTangent=1.2198, interpMode=RCIM_Cubic, leaveTangent=1.2198, time=0.3527, value=1}
         {arriveTangent=-1.0217, interpMode=RCIM_Cubic, leaveTangent=-1.0217, time=0.558, value=1}
         {arriveTangent=-0.9848, interpMode=RCIM_Cubic, leaveTangent=-0.9848, time=0.8029, value=0.54}
         <at 0.8615 0.7011>
         <at 1 0.7>
        ]}
       GreenCurve {keys [
         <at 0.0043 0.4211>
         {arriveTangent=0.9841, interpMode=RCIM_Cubic, leaveTangent=0.9841, time=0.1789, value=0.5785}
         <at 0.2812 0.6936>
         {arriveTangent=1.1069, interpMode=RCIM_Cubic, leaveTangent=1.1069, time=0.3527, value=1}
         {arriveTangent=-0.8536, interpMode=RCIM_Cubic, leaveTangent=-0.8536, time=0.558, value=1}
         {arriveTangent=-1.3917, interpMode=RCIM_Cubic, leaveTangent=-1.3917, time=0.8029, value=0.6157}
         <at 0.8615 0.5776>
         <at 1 0.4138>
        ]}
       RedCurve {keys [
         <at 0.0043 0.3653>
         {arriveTangent=1.9379, interpMode=RCIM_Cubic, leaveTangent=1.9379, time=0.1789, value=0.5542}
         <at 0.2812 0.902>
         {arriveTangent=0.3542, interpMode=RCIM_Cubic, leaveTangent=0.3542, time=0.3527, value=1}
         {arriveTangent=-0.2221, interpMode=RCIM_Cubic, leaveTangent=-0.2221, time=0.558, value=1}
         {arriveTangent=-1.476, interpMode=RCIM_Cubic, leaveTangent=-1.476, time=0.8029, value=0.9}
         <at 0.8615 0.552>
         <at 1 0.3566>
        ]}
      }}
     CurveIndex=Linked() {linkedVariable {name=NPC.FortniteNPC.RemappedTimeOfDay_0-1, type=NiagaraFloat}}
    )) {isSetParameters=true}
   &FN_Sprite_Attributes(
    Color=&MakeLinearColorFromVectorAndFloat(
     "Vector (RGB)"=&MakeVectorFromLinearColorRGB(Color=&MultiplyLinearColor(
       Linear Color=$Emitter.MasterColorA
       Scale Factor=&MakeVector4_FromLinearColor(LinearColor=$Particles.TODColor)
      ))
     "Float (Alpha)"=&FloatFromCurve(
      FloatCurve=DataInterface() {propertyValues {Curve {
         defaultValue=1
         keys [
          {arriveTangent=3.8054, interpMode=RCIM_Cubic, leaveTangent=3.8054, tangentMode=RCTM_User, time=0, value=0.6148}
          {arriveTangent=1.4463, interpMode=RCIM_Cubic, leaveTangent=1.4463, time=0.0551, value=0.9324}
          {interpMode=RCIM_Cubic, tangentMode=RCTM_User, time=0.2663, value=1}
          {arriveTangent=-1.1447, interpMode=RCIM_Cubic, leaveTangent=-1.1447, tangentMode=RCTM_User, time=1, value=0}
         ]
        }}}
      Scale Curve=0.1
     )
    )
    Create Color Attribute=true
    Create Sprite Rotation Attribute=true
    Create SpriteSize Attribute=true
    Module Output:ENiagara_AttributeSamplingApplyOutput=::NewEnumerator1
    SpriteRotation:NiagaraFloat=&FloatFromCurve(
     FloatCurve=DataInterface() {propertyValues {Curve {keys [<at 0 1>,<at 1 0>]}}}
     Scale Curve=&RandomRangeFloat(Minimum=-5, Maximum=5)
    )
    SpriteSize=&Vector2DFromFloat(Value=&Multiply_Float(
      A=$Emitter.MasterSizeMultiplier
      B=&FloatFromCurve(
       FloatCurve=DataInterface() {propertyValues {Curve {
          defaultValue=1
          keys [
           {arriveTangent=1.0221, interpMode=RCIM_Cubic, leaveTangent=1.0221, tangentMode=RCTM_User, time=0, value=0.7}
           {interpMode=RCIM_Cubic, tangentMode=RCTM_User, time=1, value=1.2}
          ]
         }}}
       Scale Curve=&RandomRangeFloat(Minimum=60, Maximum=80)
      )
     ))
   )
   &FN_Sprite_Attributes(
    Color=&MakeLinearColorFromVectorAndFloat(
     "Vector (RGB)"=&MakeVectorFromLinearColorRGB(Color=&MultiplyLinearColor(
       Linear Color=$Emitter.MasterColorA
       Scale Factor=&MakeVector4_FromLinearColor(LinearColor=$Particles.TODColor)
      ))
     "Float (Alpha)"=&FloatFromCurve(
      FloatCurve=DataInterface() {propertyValues {Curve {
         defaultValue=1
         keys [
          {interpMode=RCIM_Cubic, tangentMode=RCTM_User, time=0, value=0.4}
          {interpMode=RCIM_Cubic, tangentMode=RCTM_User, time=0.5, value=1}
          {arriveTangent=-5.7747, interpMode=RCIM_Cubic, leaveTangent=-5.7747, tangentMode=RCTM_User, time=1, value=0}
         ]
        }}}
      Scale Curve=0.1
     )
    )
    Create Color Attribute=true
    Create Position Attribute=true
    Create SpriteSize Attribute=true
    Module Output:ENiagara_AttributeSamplingApplyOutput=::NewEnumerator1
    Position:NiagaraPosition=/Niagara/DynamicInputs/Helpers/SimulationPosition()
    SpriteSize=&Multiply_Vector2D_ByFloat(Float=$Emitter.MasterSizeMultiplier, Vector2D=<140 140>)
   ) {module=FN_Sprite_Attributes001}
   &FN_Sprite_Attributes(
    Color=&MakeLinearColorFromVectorAndFloat(
     "Vector (RGB)"=&MakeVectorFromLinearColorRGB(Color=&MultiplyLinearColor(
       Linear Color=$Emitter.MasterColorA
       Scale Factor=&MakeVector4_FromLinearColor(LinearColor=$Particles.TODColor)
      ))
     "Float (Alpha)"=&FloatFromCurve(
      FloatCurve=DataInterface() {propertyValues {Curve {
         defaultValue=1
         keys [
          {interpMode=RCIM_Cubic, tangentMode=RCTM_User, time=0, value=0.4}
          {interpMode=RCIM_Cubic, tangentMode=RCTM_User, time=0.3091, value=1}
          {arriveTangent=-3.4033, interpMode=RCIM_Cubic, leaveTangent=-3.4033, tangentMode=RCTM_User, time=1, value=0}
         ]
        }}}
      Scale Curve=0.5
     )
    )
    Create Color Attribute=true
    Create Facing Attribute=true
    Create SpriteSize Attribute=true
    Facing:Vector3f=&RandomRangeVector(
     Minimum=&ReturnMeshOrientationAxis(Mesh Orientation Axis=::XAxis, Mesh Orientation Quat=$Particles.MeshOrientation)
     Maximum=&ReturnMeshOrientationAxis(Mesh Orientation Axis=::YAxis, Mesh Orientation Quat=$Particles.MeshOrientation)
    )
    Module Output:ENiagara_AttributeSamplingApplyOutput=::NewEnumerator1
    SpriteSize=&Multiply_Vector2D_ByFloat(
     Float=$Emitter.MasterSizeMultiplier
     Vector2D=&Vector2DFromFloat(Value=&RandomRangeFloat(Minimum=2, Maximum=5))
    )
   ) {module=FN_Sprite_Attributes002}
   &FN_Sprite_Attributes(
    Color=&MakeLinearColorFromVectorAndFloat(
     "Vector (RGB)"=&MakeVectorFromLinearColorRGB(Color=&MultiplyLinearColor(
       Linear Color=$Emitter.MasterColorB
       Scale Factor=&MakeVector4_FromLinearColor(LinearColor=$Particles.TODColor)
      ))
     "Float (Alpha)"=&FloatFromCurve(FloatCurve=DataInterface() {propertyValues {Curve {
         defaultValue=1
         keys [
          {interpMode=RCIM_Cubic, tangentMode=RCTM_User, time=0, value=0.4}
          {interpMode=RCIM_Cubic, tangentMode=RCTM_User, time=0.3091, value=1}
          {arriveTangent=-3.4033, interpMode=RCIM_Cubic, leaveTangent=-3.4033, tangentMode=RCTM_User, time=1, value=0}
         ]
        }}})
    )
    Create Color Attribute=true
    Create SpriteSize Attribute=true
    Module Output:ENiagara_AttributeSamplingApplyOutput=::NewEnumerator1
    SpriteSize=&Multiply_Vector2D_ByFloat(
     Float=$Emitter.MasterSizeMultiplier
     Vector2D=&Vector2DFromFloat(Value=&RandomRangeFloat(Minimum=5, Maximum=10))
    )
   ) {module=FN_Sprite_Attributes003}
   &FN_Sprite_Attributes(
    Color=&MultiplyLinearColor(
     Linear Color=$Emitter.MasterColorA
     Scale Factor=&MakeVector4_FromLinearColor(LinearColor=$Particles.TODColor)
    )
    Create Color Attribute=true
    Create Facing Attribute=true
    Create Sprite Rotation Attribute=true
    Create SpriteSize Attribute=true
    Facing:Vector3f=&RandomRangeVector(
     Minimum=&ReturnMeshOrientationAxis(Mesh Orientation Axis=::XAxis, Mesh Orientation Quat=$Particles.MeshOrientation)
     Maximum=&ReturnMeshOrientationAxis(Mesh Orientation Axis=::YAxis, Mesh Orientation Quat=$Particles.MeshOrientation)
    )
    Module Output:ENiagara_AttributeSamplingApplyOutput=::NewEnumerator1
    SpriteRotation:NiagaraFloat=&FloatFromCurve(
     FloatCurve=DataInterface() {propertyValues {Curve {
        defaultValue=1
        keys [{interpMode=RCIM_Cubic, tangentMode=RCTM_User, time=0, value=0}, {interpMode=RCIM_Cubic, tangentMode=RCTM_User, time=1, value=1}]
       }}}
     Scale Curve=&SelectFloatFromArray(
      Array Sampling Mode=::NewEnumerator0
      Direct Array Index=?
      Float Selection Array=DataInterface() {propertyValues {FloatData [50,-50,80,-80,100,-100]}}
      Random Seed=0
     )
    )
    SpriteSize=&Multiply_Vector2D_ByFloat(
     Float=$Emitter.MasterSizeMultiplier
     Vector2D=&Vector2DFromFloat(Value=&FloatFromCurve(
       FloatCurve=DataInterface() {propertyValues {Curve {
          defaultValue=1
          keys [
           {tangentMode=RCTM_User, time=0, value=1}
           {arriveTangent=-0.0405, interpMode=RCIM_Cubic, leaveTangent=-0.0405, tangentMode=RCTM_User, time=0.7009, value=1}
           {arriveTangent=-8.0089, interpMode=RCIM_Cubic, leaveTangent=-8.0089, tangentMode=RCTM_User, time=1, value=0}
          ]
         }}}
       Scale Curve=&RandomRangeFloat(Minimum=5, Maximum=10)
      ))
    )
   ) {module=FN_Sprite_Attributes004}
  ]
  rendererClasses [&NiagaraSpriteRendererProperties]
  renderers [
   {rendererClass=&NiagaraSpriteRendererProperties}
   {rendererClass=&NiagaraSpriteRendererProperties, rendererIndex=1}
   {rendererClass=&NiagaraSpriteRendererProperties, rendererIndex=2}
   {rendererClass=&NiagaraSpriteRendererProperties, rendererIndex=3}
   {rendererClass=&NiagaraSpriteRendererProperties, rendererIndex=4}
  ]
 }}
summary {
 systemName=NS_Default_PropSpawn_Frost
 emitters [{emitterName=PropSpawn_Frost, rendererClasses [&NiagaraSpriteRendererProperties]}]
 userVariables []
}
